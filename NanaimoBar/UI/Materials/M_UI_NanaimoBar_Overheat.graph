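# Translucent UI material for the NanaimoBar overheat gauge, drawing a radial progress arc with a frame and glow.
# Packed distance fields in the fuel texture's red and blue channels supply the outer frame, shadow and inner glow masks.
# RadialArc functions, a circle outline and a stepped gradient build the fill masks, shared through named reroutes.
# Progress blends three overheat colors, switching to a sine-blinking full-charge color near the charge limit.
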Material /NanaimoBar/UI/Materials/M_UI_NanaimoBar_Overheat
graph {
 properties {blendMode=BLEND_Translucent, materialDomain=MD_UI, pixelDepthOffsetMode=PDOM_Legacy, usedShadingModels=Default Lit}
 nodes [
  Cosine(None=Multiply(A=Time(), B=Constant() {r=0.162}) {constB=0.648})
  n1 = NamedRerouteUsage()
  n2 = NamedRerouteUsage()
  n3 = Max(A=NamedRerouteUsage(), B=NamedRerouteDeclaration(None=Multiply(A=@n1, B=@n2)) {name=PipBg, nodeColor=<rgba 0 1 0.4141 1>})
  Multiply(A=@n3, B=ScalarParameter() {defaultValue=1, parameterName=ChargeOpacity})
  NamedRerouteDeclaration(None=MaterialFunctionCall() {materialFunction=&GetUserInterfaceUV}.Normalized UV) {name=UV, nodeColor=<rgba 0 0 0 1>}
  n4 = TextureSample(UVs=NamedRerouteUsage()) {texture=/NanaimoBar/UI/Textures/T_UI_NanaimoBar_Fuel}
  n5 = MaterialFunctionCall!(RGBA Channels=@n4.R, Glow Min=Constant() {r=0.25}, Glow Max=Constant() {r=0.5}, Outline Thickness) {materialFunction=/FortUILibrary/MaterialFunctions/Basics/MF_UI_ChannelPackedDistanceField}
  NamedRerouteDeclaration(None=Add(A=SmoothStep(Value=@n4.R) {constMin=0.2}, B=NamedRerouteDeclaration(None=@n5.Fill) {name=OuterFrame_Mask, nodeColor=<rgba 0 0 0 1>})) {name=OuterFrame_Shadow, nodeColor=<rgba 0 0 0 1>}
  n6 = MaterialFunctionCall() {materialFunction=&GetUserInterfaceUV}
  n7 = Constant() {r=2.6085}
  n8 = ComponentMask(None=Subtract(A=Rotator(Coordinate=@n6.Normalized UV, Time=@n7), B=ScalarParameter() {defaultValue=0.232, parameterName=Gradient_Position})) {r=true}
  NamedRerouteDeclaration(None=Saturate(None=Multiply(A=@n8))) {name=ColorGradient, nodeColor=<rgba 0 1 0.7422 1>}
  n9 = MaterialFunctionCall(OutlineThickness=ScalarParameter() {defaultValue=0.01, group=AmmoCounter, parameterName=RadialThickness, sliderMax=1}) {materialFunction=&MF_UI_Circle}
  NamedRerouteDeclaration(None=OneMinus(None=SmoothStep(Value=@n9.SDFOutline) {constMax=0.01})) {name=Mask_Radial, nodeColor=<rgba 0 0 0 1>}
  n10 = MaterialFunctionCall!(
   UV=MaterialFunctionCall(UVs=NamedRerouteUsage(), "Rotation Angle (0-1)"=Constant() {r=0.5}) {materialFunction=&CustomRotator}.Rotated Values
   "Arc Start Angle (0-1)"=ScalarParameter() {defaultValue=0.97, parameterName=Start Angle, sliderMax=1, sortPriority=31}
   "Arc End Angle (0-1)"=ScalarParameter() {defaultValue=0.25, parameterName=End Angle, sliderMax=1}
   "Fill Ratio (0-1)"=NamedRerouteUsage()
   Fill Clockwise
  ) {materialFunction=/FortUILibrary/MaterialFunctions/Basics/MF_UI_RadialArc}
  n11 = SmoothStep(Value=@n10.Radial Fill) {constMax=0.001}
  NamedRerouteDeclaration(None=Multiply(A=NamedRerouteUsage(), B=@n11)) {name=Mask_OverheatProgress, nodeColor=<rgba 0 0 0 1>}
  n12 = TextureCoordinate()
  n13 = Constant2Vector() {g=-0.12, r=-0.146}
  NamedRerouteDeclaration(None=MaterialFunctionCall(UVs=Add(A=@n12, B=@n13), Texture Scale=Constant() {r=0.36}) {materialFunction=&ScaleUVsByCenter}.UVs) {name=TextureUVs, nodeColor=<rgba 0 0 0 1>}
  n14 = NamedRerouteDeclaration(None=MaterialFunctionCall(RGBA Channels=@n4.B, Glow Min=Constant() {r=0.1}, Glow Max=Constant() {r=0.8}) {materialFunction=&MF_UI_ChannelPackedDistanceField}.Fill) {name=InnerMask, nodeColor=<rgba 0 0 0 1>}
  n15 = NamedRerouteUsage()
  NamedRerouteDeclaration(None=Saturate(None=Multiply(A=@n14, B=@n15))) {name=Overheat Reached Glow, nodeColor=<rgba 0 0 0 1>}
  n16 = Constant3Vector()
  n17 = VectorParameter() {parameterName=FrameColor, defaultValue=<rgba 1 1 1 1>}
  n18 = NamedRerouteUsage()
  NamedRerouteDeclaration(None=LinearInterpolate(A=NamedRerouteUsage(), B=LinearInterpolate(A=@n16, B=@n17.RGB, Alpha=@n18), Alpha=Ceil(None=@n18))) {name=Outer Frame Color, nodeColor=<rgba 0 0 0 1>}
  n19 = VectorParameter() {group=Color, parameterName=OverheatColorC, defaultValue=<rgba 0.3185 0.807 0.0116 1>}
  n20 = LinearInterpolate(
   A=VectorParameter() {group=Color, parameterName=OverheatColorB, defaultValue=<rgba 0.3185 0.807 0.0116 1>}.RGB
   B=VectorParameter() {group=Color, parameterName=OverheatColorA, defaultValue=<rgba 0.3185 0.807 0.0116 1>}.RGB
   Alpha=MaterialFunctionCall!(
    Input=ScalarParameter() {defaultValue=0.5, group=Color, parameterName=Progress, sliderMax=1}
    Input Low=ScalarParameter() {defaultValue=0.25, group=Color, parameterName=ColorBStartProgress}
    Input High=ScalarParameter() {defaultValue=0.5, group=Color, parameterName=ColorAEndProgress}
   ) {materialFunction=/Engine/Functions/Engine_MaterialFunctions03/Math/RemapValueRangeNormalized}.Result
  )
  n21 = MaterialFunctionCall(
   Input=ScalarParameter() {defaultValue=0.5, group=Color, parameterName=Progress}
   Input Low=ScalarParameter() {defaultValue=0.75, group=Color, parameterName=ColorBEndProgress}
   Input High=ScalarParameter() {defaultValue=0.5, group=Color, parameterName=ColorCStartProgress}
  ) {materialFunction=&RemapValueRangeNormalized}
  n22 = ScalarParameter() {defaultValue=0.5, group=Color, parameterName=Progress, sliderMax=1}
  n23 = ScalarParameter() {defaultValue=1, group=Color, parameterName=FullChargeStart}
  n24 = ScalarParameter() {defaultValue=0.99, group=Color, parameterName=isChargingEnd}
  NamedRerouteDeclaration(None=LinearInterpolate(
    A=NamedRerouteUsage()
    B=LinearInterpolate(A=@n19.RGB, B=@n20, Alpha=@n21.Result)
    Alpha=MaterialFunctionCall(Input=@n22, Input Low=@n23, Input High=@n24) {materialFunction=&RemapValueRangeNormalized}.Result
   )) {name=Overheat Color, nodeColor=<rgba 0 0 0 1>}
  n25 = MaterialFunctionCall(
   UV=MaterialFunctionCall(UVs=NamedRerouteUsage(), "Rotation Angle (0-1)"=Constant() {r=0.5}) {materialFunction=&CustomRotator}.Rotated Values
   "Arc Start Angle (0-1)"=Constant() {r=0.97}
   "Arc End Angle (0-1)"=Constant() {r=0.25}
   "Fill Ratio (0-1)"=NamedRerouteUsage()
  ) {materialFunction=&MF_UI_RadialArc}
  NamedRerouteDeclaration(None=Saturate(None=Multiply(A=@n25.Radial Fill) {constB=50})) {name=Progress Radial Fill, nodeColor=<rgba 0 0 0 1>}
  n26 = MaterialFunctionCall!("Value (0-1)"=OneMinus(None=@n10.Arc Mask), Steps) {materialFunction=/FortUILibrary/MaterialFunctions/Basics/MF_UI_StepRatio}
  n27 = Multiply(A=Saturate(None=Add(A=@n26.Stepped Gradient) {constB=0.15})) {constB=0.8}
  NamedRerouteDeclaration(None=Multiply(A=Max(A=@n11, B=@n27)) {constB=0.8}) {name=OverheatProgress, nodeColor=<rgba 0 0 0 1>}
  NamedRerouteDeclaration(None=OneMinus(None=ScalarParameter() {defaultValue=0.5, group=Color, parameterName=Progress, sliderMax=1})) {name=Progress, nodeColor=<rgba 0 0 0 1>}
  n28 = Multiply(A=Time(), B=ScalarParameter() {defaultValue=5, parameterName=blink Speed})
  NamedRerouteDeclaration(None=LinearInterpolate(
    A=VectorParameter() {group=Color, parameterName=OverheatColorC, defaultValue=<rgba 0.3185 0.807 0.0116 1>}.RGB
    B=VectorParameter() {group=Color, parameterName=FullChargeColor, defaultValue=<rgba 0.807 0.807 0.807 1>}.RGB
    Alpha=Sine(None=@n28)
   )) {name=FullChargeColor, nodeColor=<rgba 1 0 0.6562 1>}
  NamedRerouteUsage()
  Saturate(None=@n3)
 ]
}
