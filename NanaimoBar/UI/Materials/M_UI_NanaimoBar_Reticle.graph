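# Translucent UI material that draws the NanaimoBar reticle: a circle, drop lines, bars, frame and side details, each with shadows.
# Shapes come from SDF math and mostly mirrored MSDF textures, combined through named reroutes into layered color and per-element opacity under ReticleOpacity.
# Progress drives a radial arc fill and edge line, alongside gradient colors with ColorDodge and LinearLight blends.
# FullFx enables a sine-pulsed bar shadow and a panning-noise bubble mask.
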Material /NanaimoBar/UI/Materials/M_UI_NanaimoBar_Reticle
graph {
 properties {bAutomaticallySetUsageInEditor=false, blendMode=BLEND_Translucent, materialDomain=MD_UI, usedShadingModels=Default Lit}
 nodes [
  n1 = Constant3Vector()
  n2 = Constant3Vector() {constant=<rgba 1 1 1 0>}
  n3 = NamedRerouteUsage()
  n4 = NamedRerouteUsage()
  n5 = Add(A=Add(A=Add(A=@n3, B=@n4), B=NamedRerouteUsage()), B=NamedRerouteUsage())
  n6 = NamedRerouteUsage()
  n7 = NamedRerouteUsage()
  n8 = Multiply(A=Power(Base=MaterialFunctionCall!(UVs, Offset, RotationCenter, RotationAngle) {materialFunction=/FortUILibrary/MaterialFunctions/Gradients/MF_UI_ReflectedGradient}.Result) {constExponent=7}) {constB=3}
  n9 = MaterialFunctionCall!(
   Blend=Multiply(A=Constant4Vector() {constant=<rgba 0.956 0.0513 1 1>}.RGBA, B=NamedRerouteUsage())
   Base=LinearInterpolate(
    A=VectorParameter() {parameterName=Color_Gradient01, defaultValue=<rgba 0.7231 0.0194 1 1>}.RGB
    B=VectorParameter() {parameterName=Color_Gradient02, defaultValue=<rgba 0.2705 0.0097 1 1>}.RGB
    Alpha=Saturate(None=@n8)
   )
  ) {materialFunction=/Engine/Functions/Engine_MaterialFunctions03/Blends/Blend_ColorDodge}
  n10 = NamedRerouteUsage()
  n11 = Multiply(A=Constant3Vector() {constant=<rgba 1 0.9215 0 1>}, B=@n10)
  LinearInterpolate(
   A=LinearInterpolate(
    A=LinearInterpolate(A=@n1, B=@n2, Alpha=@n5)
    B=VectorParameter() {parameterName=Color_BarShadow, defaultValue=<rgba 1 0 0.0992 1>}.RGB
    Alpha=Multiply(A=@n6, B=@n7)
   )
   B=LinearInterpolate(
    A=@n9.Result
    B=MaterialFunctionCall!(Base=@n11, Blend=@n9.Result) {materialFunction=/Engine/Functions/Engine_MaterialFunctions03/Blends/Blend_LinearLight}.Result
    Alpha=@n10
   )
   Alpha=Multiply(A=NamedRerouteUsage(), B=@n7)
  )
  n12 = NamedRerouteUsage()
  n13 = ScalarParameter() {defaultValue=0.4, parameterName=ShadowOpacity}
  n14 = Multiply(A=Add(A=NamedRerouteUsage(), B=Multiply(A=@n12, B=@n13)), B=ScalarParameter() {defaultValue=0.5, parameterName=BracketsOpacity}) {constB=0.4}
  n15 = NamedRerouteUsage()
  n16 = NamedRerouteUsage()
  n17 = Multiply(A=Saturate(None=Add(A=@n15, B=@n16)), B=ScalarParameter() {defaultValue=1, parameterName=DropLinesOpacity})
  n18 = NamedRerouteUsage()
  n19 = ScalarParameter() {defaultValue=1, parameterName=CircleOpacity}
  n20 = NamedRerouteUsage()
  n21 = ScalarParameter() {defaultValue=1, parameterName=BarShadowOpacity}
  n22 = NamedRerouteUsage()
  n23 = Multiply(A=NamedRerouteUsage(), B=@n22)
  n24 = Add(
   A=Add(A=Add(A=@n14, B=@n17), B=Multiply(A=@n18, B=@n19))
   B=Add(A=Multiply(A=@n20, B=@n21), B=LinearInterpolate(A=@n22, B=@n23) {constAlpha=0.6})
  )
  n25 = NamedRerouteUsage()
  n26 = Multiply(A=NamedRerouteUsage()) {constB=0.4}
  n27 = Multiply(A=Saturate(None=Add(A=@n25, B=@n26)), B=ScalarParameter() {defaultValue=0.8, parameterName=SideDetailsOpacity}) {constB=0.7}
  Multiply(A=Saturate(None=Add(A=@n24, B=@n27)), B=ScalarParameter() {defaultValue=1, parameterName=ReticleOpacity})
  n28 = MaterialFunctionCall(
   UVs=NamedRerouteUsage()
   Size=ScalarParameter() {defaultValue=0.093, parameterName=CircleSize}
   OutlineThickness=ScalarParameter() {parameterName=CircleThickness}
  ) {materialFunction=&MF_UI_Circle}
  NamedRerouteDeclaration(None=OneMinus(None=SmoothStep(Value=@n28.SDFOutline) {constMax=0.015, constMin=0.005})) {name=Circle, nodeColor=<rgba 0 0 0 1>}
  NamedRerouteDeclaration(None=OneMinus(None=SmoothStep(Value=@n28.SDFOutline) {constMax=0.02, constMin=0.006})) {name=CircleShadow, nodeColor=<rgba 0 0 0 1>}
  n29 = NamedRerouteUsage()
  n30 = ComponentMask(None=@n29) {r=true}
  n31 = Add(A=ComponentMask(None=@n29) {g=true}) {constB=-0.41}
  n32 = TextureSample(UVs=MaterialFunctionCall(UVs=AppendVector(A=@n30, B=@n31), Texture Scale=ScalarParameter() {defaultValue=0.55, parameterName=LinesSize}) {materialFunction=&ScaleUVsByCenter}.UVs) {samplerType=SAMPLERTYPE_LinearGrayscale, texture=/NanaimoBar/UI/Textures/T_UI_NanaimoBar_DropLines}
  NamedRerouteDeclaration(None=MaterialFunctionCall(DistanceField=@n32.R, Threshold=Constant() {r=0.68}) {materialFunction=&MF_UI_AntiAliasing}.Outer) {name=DropLines, nodeColor=<rgba 0 0 0 1>}
  NamedRerouteDeclaration(None=Multiply(A=SmoothStep(Value=@n32.R) {constMin=0.5}) {constB=0.4}) {name=DropLinesShadow, nodeColor=<rgba 0 0 0 1>}
  n33 = NamedRerouteUsage()
  n34 = Add(
   A=MaterialFunctionCall!(UV=@n33, Swizzle UV) {materialFunction=/FortUILibrary/MaterialFunctions/Basics/MF_UI_Symmetry_2x2_New}.Symmetry
   B=Constant2Vector() {g=-0.2, r=-0.28}
  )
  n35 = Constant() {r=0.45}
  n36 = MaterialFunctionCall!(MSDF RGB=TextureSample(UVs=MaterialFunctionCall(UVs=@n34, Texture Scale=@n35) {materialFunction=&ScaleUVsByCenter}.UVs) {texture=/NanaimoBar/UI/Textures/T_UI_NanaimoBar_Bracket_MSDF}.RGB) {materialFunction=/FortUILibrary/MaterialFunctions/Shapes/Modifiers/MF_UI_MSDFtoSDF}
  n37 = MaterialFunctionCall(DistanceField=@n36.SDF, Threshold=Constant() {r=0.3}) {materialFunction=&MF_UI_AntiAliasing}
  NamedRerouteDeclaration(None=@n37.Outer) {name=Bar, nodeColor=<rgba 0 0 0 1>}
  n38 = Constant() {r=0.28}
  n39 = MaterialFunctionCall(Speed=Constant() {r=0.3}) {materialFunction=&DebugTimeSine}
  n40 = MaterialFunctionCall(
   DistanceField=@n36.SDF
   Threshold=LinearInterpolate(A=@n38, B=LinearInterpolate(A=@n38, Alpha=@n39.Result) {constA=0.28, constB=0.25}, Alpha=ScalarParameter() {parameterName=FullFx})
  ) {materialFunction=&MF_UI_AntiAliasing}
  NamedRerouteDeclaration(None=Saturate(None=Subtract(A=@n40.Outer, B=@n37.Outer))) {name=BarShadow, nodeColor=<rgba 0 0 0 1>}
  n41 = Add(
   A=MaterialFunctionCall!(UV=NamedRerouteUsage(), Mirror Y, Swizzle UV) {materialFunction=/FortUILibrary/MaterialFunctions/Basics/MF_UI_Symmetry_2x1_New}.Symmetry
   B=Constant2Vector() {r=-0.12}
  )
  n42 = Constant2Vector() {g=0.3, r=0.6}
  n43 = MaterialFunctionCall(MSDF RGB=TextureSample(UVs=MaterialFunctionCall(UVs=@n41, Texture Scale=@n42) {materialFunction=&ScaleUVsByCenter}.UVs) {texture=/NanaimoBar/UI/Textures/T_UI_NanaimoBar_SideDetail_MSDF}.RGB) {materialFunction=&MF_UI_MSDFtoSDF}
  NamedRerouteDeclaration(None=MaterialFunctionCall(DistanceField=@n43.SDF, Threshold=Constant() {r=0.35}) {materialFunction=&MF_UI_AntiAliasing}.Outer) {name=SideDetails, nodeColor=<rgba 0 0 0 1>}
  NamedRerouteDeclaration(None=MaterialFunctionCall(DistanceField=@n43.SDF, Threshold=Constant() {r=0.3}) {materialFunction=&MF_UI_AntiAliasing}.Outer) {name=SideDetailsShadow, nodeColor=<rgba 0 0 0 1>}
  n44 = ScalarParameter() {defaultValue=1, parameterName=Progress}
  n45 = MaterialFunctionCall!(
   UV
   "Arc Start Angle (0-1)"
   "Arc End Angle (0-1)"
   "Fill Ratio (0-1)"=If(A=@n44, "A > B"=LinearInterpolate(Alpha=@n44) {constA=0.55, constB=0.088}, "A < B"=LinearInterpolate(Alpha=@n44) {constA=0.915, constB=0.43}) {constB=0.5}
   Fill Clockwise
  ) {materialFunction=/FortUILibrary/MaterialFunctions/Basics/MF_UI_RadialArc}
  n46 = SmoothStep(Value=@n45.Centered Fill) {constMax=0.0005}
  NamedRerouteDeclaration(None=@n46) {name=Progress, nodeColor=<rgba 0 0 0 1>}
  n47 = SmoothStep(Max=If(A=@n44, "A > B"=Constant() {r=0.0005}, "A < B"=Constant() {r=0.05}) {constB=0.99}, Value=@n45.Centered Fill) {constMax=0.05}
  NamedRerouteDeclaration(None=Saturate(None=Subtract(A=@n46, B=@n47))) {name=ProgressLine, nodeColor=<rgba 0 0 0 1>}
  n48 = Add(
   A=MaterialFunctionCall(UV=@n33) {materialFunction=&MF_UI_Symmetry_2x2_New}.Symmetry
   B=Constant() {r=-0.23}
  )
  n49 = Constant() {r=1.25}
  n50 = MaterialFunctionCall(MSDF RGB=TextureSample(UVs=MaterialFunctionCall(UVs=@n48, Texture Scale=@n49) {materialFunction=&ScaleUVsByCenter}.UVs) {texture=/NanaimoBar/UI/Textures/T_UI_NanaimoBar_Bar_MSDF}.RGB) {materialFunction=&MF_UI_MSDFtoSDF}
  n51 = MaterialFunctionCall(DistanceField=@n50.SDF, Threshold=Constant() {r=0.27}) {materialFunction=&MF_UI_AntiAliasing}
  NamedRerouteDeclaration(None=@n51.Outer) {name=Frame, nodeColor=<rgba 0 0 0 1>}
  n52 = MaterialFunctionCall(DistanceField=@n36.SDF, Threshold=Constant() {r=0.3}, Smoothing=LinearInterpolate(Alpha=@n39.Result) {constA=0.2, constB=0.6}) {materialFunction=&MF_UI_AntiAliasing}
  n53 = NamedRerouteUsage()
  n54 = TextureSample(UVs=Panner(Coordinate=Multiply(A=@n53) {constB=4}) {speedX=0.02, speedY=0.2}) {texture=/NanaimoBar/UI/Textures/T_UI_NanaimoBar_Noise}
  n55 = ScalarParameter() {parameterName=FullFx}
  NamedRerouteDeclaration(None=Multiply(A=Multiply(A=@n52.Outer, B=@n54.RGB), B=LinearInterpolate(Alpha=@n55) {constAlpha=0}) {constB=0}) {name=BubbleFx, nodeColor=<rgba 0 0 0 1>}
  n56 = MaterialFunctionCall(DistanceField=@n50.SDF, Threshold=Constant() {r=0.26}) {materialFunction=&MF_UI_AntiAliasing}
  NamedRerouteDeclaration(None=Saturate(None=Subtract(A=@n56.Outer, B=@n51.Outer))) {name=FrameShadow, nodeColor=<rgba 0 0 0 1>}
  NamedRerouteDeclaration(None=MaterialFunctionCall() {materialFunction=&GetUserInterfaceUV}.Normalized UV) {name=UV, nodeColor=<rgba 0 0 0 1>}
 ]
}
