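# Charged-weapon reticle widget for the NanaimoBar weapon, showing charge progress and a hold-to-charge hint.
# Charge updates set Progress on the reticle's dynamic material and trigger one-shot half and full-charge animations, guarded by booleans that firing resets.
# Intro, charge and shot animations all keyframe material parameters on a single image.
# On construct and trigger press, touch players see the hint until a restartable three-second timer hides it.
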WidgetBlueprint /NanaimoBar/UI/WBP_NanaimoBar_Reticle
animations [
 Anim_Intro {
  binding_count=1
  length_seconds=0.5
  object=&self:Anim_Intro
  sequence {
   bindings [Image_Reticle {
     id=#1
     possessed_class=Image
     tracks [{
       type=MovieSceneWidgetMaterialTrack
       sections [{
         type=MovieSceneParameterSection
         end_frame=15
         start_frame=0
         channels [ReticleOpacity {type=MovieSceneScriptingFloatChannel, keys [[0,0],[15,1]]}]
        }]
      }]
    }]
   display_rate {numerator=30}
  }
 }
 Anim_Fullcharge {
  binding_count=1
  length_seconds=0.25
  object=&self:Anim_Fullcharge
  sequence {bindings [Image_Reticle {
     id=#2
     possessed_class=Image
     tracks [{
       type=MovieSceneWidgetMaterialTrack
       sections [{
         type=MovieSceneParameterSection
         end_frame=5
         start_frame=0
         channels [
          FullFx {type=MovieSceneScriptingFloatChannel, keys [[0,0],[5,1]]}
          CircleSize {type=MovieSceneScriptingFloatChannel, keys [[0,0.149],[3,0.015]]}
          CircleThickness {type=MovieSceneScriptingFloatChannel, keys [[0,-0.002],[3,0.008]]}
          DropLinesOpacity {type=MovieSceneScriptingFloatChannel, keys [[0,1],[3,0]]}
          Color_Gradient01R {type=MovieSceneScriptingFloatChannel, keys [[0,0.7231],[5,1]]}
          Color_Gradient01G {type=MovieSceneScriptingFloatChannel, keys [[0,0.0194],[5,0.2]]}
          Color_Gradient01B {type=MovieSceneScriptingFloatChannel, const [1,2]}
          Color_Gradient01A {type=MovieSceneScriptingFloatChannel, const [1,1]}
         ]
        }]
      }]
    }]}
 }
 Anim_Halfcharge {
  binding_count=1
  length_seconds=0.25
  object=&self:Anim_Halfcharge
  sequence {bindings [Image_Reticle {
     id=#3
     possessed_class=Image
     tracks [{
       type=MovieSceneWidgetMaterialTrack
       sections [{
         type=MovieSceneParameterSection
         end_frame=5
         start_frame=0
         channels [
          Color_Gradient02R {type=MovieSceneScriptingFloatChannel, keys [[0,0.2705],[3,0.723],[5,0.2705]]}
          Color_Gradient02G {type=MovieSceneScriptingFloatChannel, keys [[0,0.0097],[3,0.019],[5,0.0097]]}
          Color_Gradient02B {type=MovieSceneScriptingFloatChannel, const [1,1]}
          Color_Gradient02A {type=MovieSceneScriptingFloatChannel, const [1,1]}
         ]
        }]
      }]
    }]}
 }
 Anim_Shooted {
  binding_count=1
  length_seconds=0.05
  object=&self:Anim_Shooted
  sequence {bindings [Image_Reticle {
     id=#4
     possessed_class=Image
     tracks [{
       type=MovieSceneWidgetMaterialTrack
       sections [{
         type=MovieSceneParameterSection
         end_frame=1
         start_frame=0
         channels [
          FullFx {type=MovieSceneScriptingFloatChannel, const [0,1]}
          CircleSize {type=MovieSceneScriptingFloatChannel, const [0.109,2]}
          CircleThickness {type=MovieSceneScriptingFloatChannel, const [-0.002,2]}
          DropLinesOpacity {type=MovieSceneScriptingFloatChannel, const [1,2]}
          Color_Gradient01R {type=MovieSceneScriptingFloatChannel, const [0.7231,1]}
          Color_Gradient01G {type=MovieSceneScriptingFloatChannel, const [0.0194,1]}
          Color_Gradient01B {type=MovieSceneScriptingFloatChannel, const [1,1]}
          Color_Gradient01A {type=MovieSceneScriptingFloatChannel, const [1,1]}
         ]
        }]
      }]
    }]}
 }
]
graphs [
 EventGraph {nodes [
   OnFullCharge() {is=title K2Node_CustomEvent}
   > n1 = Get!(execute:Exec) -> (then:Exec, else:Exec, isFullCharged:bool)
   Set isFullCharged(execute=@n1.else, isFullCharged:bool=true) -> (Output_Get:bool)
   > PlayAnimation(InAnimation=Get Anim_Fullcharge() -> (Anim_Fullcharge:Widget Animation Object Reference))
   OnPressTrigger!() -> (OutputDelegate, then, Weapon:Fort Weapon Object Reference) {is=title K2Node_CustomEvent}
   > ShowPrompt()
   OnHalfCharge() {is=title K2Node_CustomEvent}
   > n2 = Get!(execute) -> (then, else, isHalfCharged:bool)
   Set isHalfCharged(execute=@n2.else, isHalfCharged:bool=true) -> (Output_Get:bool)
   > PlayAnimation(InAnimation=Get Anim_Halfcharge() -> (Anim_Halfcharge:Widget Animation Object Reference))
   Event OnWeaponSingleFire() {is=title K2Node_Event}
   > Set isHalfCharged(isHalfCharged=false)
   > Set isFullCharged(isFullCharged=false)
   > PlayAnimation(InAnimation=Get Anim_Shooted() -> (Anim_Shooted:Widget Animation Object Reference))
   Event Destruct()
   > n3 = Get!(execute) -> (then, else, BoundWeapon:Fort Weapon Object Reference)
   n4 = Create Event()
   > n5 = Unbind Event from On Press Trigger(self:Fort Weapon Object Reference=@n3, Delegate=@n4) {is=title K2Node_RemoveDelegate}
   n6 = Create Event()
   n7 = GetOwningPlayer()
   n8 = Cast To FortPlayerController(execute=@n14.then, Object=@n7)
   > n9 = Set OwningFortPlayerController(OwningFortPlayerController:Fort Player Controller Object Reference=@n8) -> (Output_Get:Fort Player Controller Object Reference)
   Unbind Event from On View Target Changed(execute=@n5.then, self:Fort Player Controller Object Reference=@n9, Delegate=@n6) {is=title K2Node_RemoveDelegate}
   n10 = Event OnUpdateChargeDisplay!() -> (OutputDelegate, then, ChargePercent:float) {is=title K2Node_Event}
   > SetScalarParameterValue(self=Get Mat_AmmoCounter() -> (Mat_AmmoCounter:Material Instance Dynamic Object Reference), ParameterName=Progress, Value=@n10.ChargePercent)
   > n11 = Branch(Condition="float >= float"(A=@n10.ChargePercent, B=1))
   > OnFullCharge() {is=title K2Node_CallFunction}
   Branch(execute=@n11.else, Condition="float >= float"(A=@n10.ChargePercent, B=0.5))
   > OnHalfCharge() {is=title K2Node_CallFunction}
   Event Construct()
   > n12 = Sequence()
   > n13 = Set BoundWeapon(BoundWeapon:Fort Weapon Object Reference=GetEquippedWeapon!(self:Fort Weapon Reticle Extension Widget Base Object Reference) -> (ReturnValue:Fort Weapon Object Reference)) -> (Output_Get:Fort Weapon Object Reference)
   > n14 = Bind Event to On Press Trigger(self:Fort Weapon Object Reference=@n13, Delegate=@n4) {is=title K2Node_AddDelegate}
   n15 = GetDynamicMaterial(execute=@n12.then_1, self=Get Image_Reticle() -> (Image_Reticle:Image Object Reference))
   > Set Mat_AmmoCounter(Mat_AmmoCounter:Material Instance Dynamic Object Reference=@n15) -> (Output_Get:Material Instance Dynamic Object Reference)
   > PlayAnimation(InAnimation=Get Anim_Intro() -> (Anim_Intro:Widget Animation Object Reference))
   > ShowPrompt()
   OnViewTargetChanged!(
    execute=@n16.then
    self
    PlayerController:Fort Player Controller Object Reference
    OldViewTarget:Actor Object Reference
    NewViewTarget:Actor Object Reference=GetViewTarget!(self:Controller Object Reference=@n7) -> (ReturnValue:Actor Object Reference)
   ) -> (then)
   OnViewTargetChanged!() -> (
    OutputDelegate
    then
    PlayerController:Fort Player Controller Object Reference
    OldViewTarget:Actor Object Reference
    NewViewTarget:Actor Object Reference
   ) {is=title K2Node_CustomEvent}
   Event OnInitialized()
   n16 = Bind Event to On View Target Changed(execute=@n9.then, self:Fort Player Controller Object Reference=@n9, Delegate=@n6) {is=title K2Node_AddDelegate}
  ]}
 HidePrompt {nodes [
   HidePrompt() {is=title K2Node_FunctionEntry}
   > SetVisibility(self=Get HoldPrompt_Text() -> (HoldPrompt_Text:Common Text Object Reference), InVisibility=Collapsed)
  ]}
 ShowPrompt {nodes [
   ShowPrompt() {is=title K2Node_FunctionEntry}
   > Branch(Condition=IsUsingInputType!(InputType:ECommonInputType Enum=Touch, PlayerController:Player Controller Object Reference=GetOwningPlayer()) -> (ReturnValue:bool))
   > Clear and Invalidate Timer by Handle(Handle=Get PromptTimerHandle() -> (PromptTimerHandle:Timer Handle Structure))
   > SetVisibility(self=Get HoldPrompt_Text(), InVisibility=SelfHitTestInvisible)
   > n1 = Set Timer by Function Name!(
    execute
    Object:Object Reference
    FunctionName:String=HidePrompt
    Time:float=3
    bLooping:bool=false
    bMaxOncePerFrame:bool=false
    InitialStartDelay:float=0
    InitialStartDelayVariance:float=0
   ) -> (then, ReturnValue:Timer Handle Structure) {is=title K2Node_CallFunction}
   > Set PromptTimerHandle(PromptTimerHandle:Timer Handle Structure=@n1) -> (Output_Get:Timer Handle Structure)
  ]}
]
variables [
 BoundWeapon {type=FortWeapon, default=null}
 isFullCharged {type=boolean}
 isHalfCharged {type=boolean}
 Mat_AmmoCounter {type=MaterialInstanceDynamic}
 NanaimboBar_Component {type=B_NanaimoBar_Weapon_C, default=null}
 OwningFortPlayerController {type=FortPlayerController, default=null}
 PromptTimerHandle {type=TimerHandle}
]
widgets {
 info {
  inheritedWidgetCount=1
  parentClass=/Script/FortniteUI.FortChargedWeaponReticleExtensionWidget
  rootWidgetClass=&Overlay
  widgetCount=5
 }
 widgets [
  Overlay_34 {widgetClassPath=&Overlay, widget=Overlay_34}
  Image_Reticle {
   widgetClassPath=&Image
   bIsVariable=true
   parent=Overlay_34
   widget=Image_Reticle
   properties {brush {
     bIsDynamicallyLoaded=false
     drawAs=Image
     imageType=FullColor
     mirroring=NoMirror
     resourceName=None
     resourceObject=/NanaimoBar/UI/Materials/MI_UI_NanaimoBar_Reticles
     tiling=NoTile
     imageSize=<180 180>
     margin=<ltrb 0 0 0 0>
     outlineSettings {
      bUseBrushTransparency=false
      roundingType=HalfHeightRadius
      width=0
      color {colorUseRule=UseColor_Specified, specifiedColor=<rgba 0 0 0 0>}
      cornerRadii=<0 0 0 0>
     }
     tintColor {colorUseRule=UseColor_Specified, specifiedColor=<rgba 1 1 1 1>}
     uVRegion {bIsValid=false, max=<0 0>, min=<0 0>}
    }}
   slot {class=OverlaySlot, properties {horizontalAlignment=HAlign_Center, verticalAlignment=VAlign_Center}}
  }
  HoldPrompt {
   widgetClassPath=&HorizontalBox
   parent=Overlay_34
   widget=HoldPrompt
   properties {renderTransform {angle=0, scale=<1 1>, shear=<0 0>, translation=<200 59>}}
   slot {class=OverlaySlot, properties {horizontalAlignment=HAlign_Center, verticalAlignment=VAlign_Center}}
  }
  HoldPrompt_Text {
   widgetClassPath=&CommonTextBlock
   bIsVariable=true
   parent=HoldPrompt
   widget=HoldPrompt_Text
   properties {
    style=/FortUILibrary/Styles/Typography/_TextStyle_TagNormal._TextStyle_TagNormal_C
    text=Hold to Charge projectile
    visibility=Collapsed
    font {
     bForceMonospaced=false
     bMaterialIsStencil=false
     fontMaterial=null
     fontObject=&BurbankSmall
     letterSpacing=0
     monospacedWidth=1
     size=15
     skewAmount=0
     typefaceFontName=Black
     outlineSettings {
      bApplyOutlineToDropShadows=false
      bMiteredCorners=false
      bSeparateFillAlpha=false
      outlineMaterial=null
      outlineSize=0
      outlineColor=<rgba 0 0 0 1>
     }
    }
    shadowOffset=<0 0>
    strikeBrush {
     bIsDynamicallyLoaded=false
     drawAs=NoDrawType
     imageType=NoImage
     mirroring=NoMirror
     resourceName=None
     resourceObject=null
     tiling=NoTile
     imageSize=<32 32>
     margin=<ltrb 0 0 0 0>
     outlineSettings {
      bUseBrushTransparency=false
      roundingType=HalfHeightRadius
      width=0
      color {colorUseRule=UseColor_Specified, specifiedColor=<rgba 0 0 0 0>}
      cornerRadii=<0 0 0 0>
     }
     tintColor {colorUseRule=UseColor_Specified, specifiedColor=<rgba 1 1 1 1>}
     uVRegion {bIsValid=false, max=<0 0>, min=<0 0>}
    }
   }
   slot {class=HorizontalBoxSlot, properties {verticalAlignment=VAlign_Center, padding=<ltrb 4 0 0 0>}}
  }
  TutorialHighlight {widgetClassPath=&TutorialHighlight}
 ]
}
